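# Editor utility widget row for the cosmetic instruments tool, comparing one named property across Fort item definitions.
# PreConstruct sets the name label and save checkbox, then builds a property view for the source item, an arrow, and one view per target item.
# EntryData picks SinglePropertyView or CustomPropertyView and hides the checkbox and arrow when saving is disallowed.
# Checkbox changes update EntryData and broadcast OnCheckboxChanged.
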EditorUtilityWidgetBlueprint /Game/ArtTools/CosmeticTech/EUW_CosmeticInstrumentsEntry
graphs [
 ConstructCustomProperty {nodes [
   n1 = ConstructCustomProperty!() -> (then, ItemDef:Fort Item Definition Object Reference) {is=title K2Node_FunctionEntry}
   > n2 = Construct Object from Class(Class=/Script/FortniteEditor.CustomPropertyView, self=Get HorizontalBox() -> (HorizontalBox:Horizontal Box Object Reference)) -> (ReturnValue:Custom Property View Object Reference)
   > n3 = AddChild(self=Get HorizontalBox(), Content=@n2)
   > n4 = Cast To HorizontalBoxSlot!(execute, Object=@n3) -> (then, CastFailed, AsHorizontal Box Slot:Horizontal Box Slot Object Reference) {is=title K2Node_DynamicCast}
   > n5 = SetSize!(execute, self:Horizontal Box Slot Object Reference=@n4, InSize_Value:float=1, InSize_SizeRule:ESlateSizeRule Enum=Fill) -> (then)
   SetObject(execute=[@n5.then, @n4.CastFailed], self=@n2, NewObject=@n1)
   > SetCustomViewPropertyAndRebuild!(
    execute
    self:Fort Cosmetics Instruments Entry Object Reference
    CustomProperty:Custom Property View Object Reference=@n2
    PropertyName:Name=Get!() -> (
     EntryData_PropertyName_2_#1:Name
     EntryData_UseSingleProperty_5_#2:bool=~
     EntryData_AllowSaving_7_#3:bool=~
     EntryData_LastCheckboxState_9_#4:bool=~
    )
   ) -> (then)
   > Return Node!(execute, PropertyWidget:Property View Base Object Reference=@n2)
  ]}
 ConstructProperty {nodes [
   n1 = ConstructProperty!() -> (then, ItemDef:Fort Item Definition Object Reference) {is=title K2Node_FunctionEntry}
   > n2 = Branch(Condition=Get() -> (
     EntryData_UseSingleProperty_5_#2=~
     EntryData_AllowSaving_7_#3=~
     EntryData_LastCheckboxState_9_#4=~
    ).EntryData_UseSingleProperty_5_#2)
   > n3 = ConstructSingleProperty!(execute, self, ItemDef:Fort Item Definition Object Reference=@n1) -> (then, PropertyWidget:Property View Base Object Reference)
   > Return Node(PropertyWidget=@n3)
   n4 = ConstructCustomProperty!(execute=@n2.else, self, ItemDef:Fort Item Definition Object Reference=@n1) -> (then, PropertyWidget:Property View Base Object Reference) {is=title K2Node_CallFunction}
   > Return Node(PropertyWidget=@n4)
  ]}
 ConstructSingleProperty {nodes [
   n1 = ConstructSingleProperty!() -> (then, ItemDef:Fort Item Definition Object Reference) {is=title K2Node_FunctionEntry}
   > n2 = Construct Object from Class(Class=&SinglePropertyView, self=Get HorizontalBox()) -> (ReturnValue:Single Property View Object Reference)
   > n3 = AddChild(self=Get HorizontalBox(), Content=@n2)
   > n4 = Cast To HorizontalBoxSlot(Object=@n3)
   > n5 = SetSize(self=@n4, InSize_Value=1, InSize_SizeRule=Fill)
   SetObject(execute=[@n5.then, @n4.CastFailed], self=@n2, NewObject=@n1)
   > SetPropertyName!(
    execute
    self:Single Property View Object Reference=@n2
    NewPropertyName:Name=Get() -> (
     EntryData_UseSingleProperty_5_#2=~
     EntryData_AllowSaving_7_#3=~
     EntryData_LastCheckboxState_9_#4=~
    )
   ) -> (then)
   > Return Node(PropertyWidget=@n2)
  ]}
 EventGraph {nodes [
   n1 = "On Check State Changed (SaveCheckbox)"!() -> (OutputDelegate, then, bIsChecked:bool) {is=title K2Node_ComponentBoundEvent}
   n2 = Get() -> (
    EntryData_UseSingleProperty_5_#2=~
    EntryData_AllowSaving_7_#3=~
    EntryData_LastCheckboxState_9_#4=~
   )
   > n3 = Set!(
    execute
    EntryData_PropertyName_2_#1:Name=@n2
    EntryData_UseSingleProperty_5_#2:bool=@n2.EntryData_UseSingleProperty_5_#2
    EntryData_AllowSaving_7_#3:bool=@n2.EntryData_AllowSaving_7_#3
    EntryData_LastCheckboxState_9_#4:bool=@n1.bIsChecked
   ) -> (then, Output_Get:Cosmetic Property Data Entry Structure)
   > Call On Checkbox Changed!(execute:Exec, self:EUW Cosmetic Instruments Entry Object Reference, EntryData:Cosmetic Property Data Entry Structure=@n3) -> (then:Exec) {is=title K2Node_CallDelegate}
   Event PreConstruct()
   > n4 = Sequence!(execute) -> (then_0, then_1, then_2:Exec, then_3:Exec)
   ConstructProperty!(
    execute=@n4.then_1
    self
    ItemDef:Fort Item Definition Object Reference=Get ItemDefToCopy() -> (ItemDefToCopy:Fort Item Definition Object Reference)
   ) -> (then, PropertyWidget:Property View Base Object Reference) {is=title K2Node_CallFunction}
   > n5 = Construct Object from Class(Class=&ScaleBox, self=Get HorizontalBox()) -> (ReturnValue:Scale Box Object Reference)
   > n6 = AddChild(self=Get HorizontalBox(), Content=@n5)
   > n7 = Cast To HorizontalBoxSlot(Object=@n6)
   > n8 = SetPadding!(
    execute
    self:Horizontal Box Slot Object Reference=@n7
    InPadding_Left:float=25
    InPadding_Top:float=0
    InPadding_Right:float=25
    InPadding_Bottom:float=0
   ) -> (then)
   n9 = Construct Object from Class(execute=[@n8.then, @n7.CastFailed], Class=&Image, self=@n5) -> (ReturnValue:Image Object Reference)
   > SetBrushFromTexture!(
    execute
    self:Image Object Reference=@n9
    Texture:Texture 2D Object Reference=/Game/NeverCook/EditorUtilities/DebugTools/AssetChangeFinder/Art/ArrowRight
    bMatchSize:bool=false
   ) -> (then)
   > AddChild(self=@n5, Content=@n9)
   > n10 = Branch(Condition=Get() -> (
     EntryData_UseSingleProperty_5_#2=~
     EntryData_AllowSaving_7_#3=~
     EntryData_LastCheckboxState_9_#4=~
    ).EntryData_AllowSaving_7_#3)
   SetVisibility(execute=@n10.else, self=@n5, InVisibility=Hidden)
   n11 = For Each Loop(
    Exec=@n4.then_2
    Array:Array of Fort Item Definition Object References=Get ItemDefsToShow() -> (ItemDefsToShow:Array of Fort Item Definition Object References)
   ) -> (Array Element:Fort Item Definition Object Reference)
   > ConstructProperty(ItemDef=@n11)
   n12 = Get() -> (
    EntryData_UseSingleProperty_5_#2=~
    EntryData_AllowSaving_7_#3=~
    EntryData_LastCheckboxState_9_#4=~
   )
   n13 = Branch(execute=@n4.then_0, Condition=@n12.EntryData_AllowSaving_7_#3)
   > n14 = SetCheckedState!(
    execute
    self:Check Box Object Reference=Get SaveCheckbox() -> (SaveCheckbox:Editor Utility Check Box Object Reference)
    InCheckedState:ECheckBoxState Enum=Select(
     Option 0:ECheckBoxState Enum=Unchecked
     Option 1:ECheckBoxState Enum=Checked
     Index=@n12.EntryData_LastCheckboxState_9_#4
    ) -> (ReturnValue:ECheckBoxState Enum)
   ) -> (then)
   Set Text(
    execute=[@n14.then, @n16.then]
    Text="To Text (Name)"!(InName:Name=Get() -> (
      EntryData_UseSingleProperty_5_#2=~
      EntryData_AllowSaving_7_#3=~
      EntryData_LastCheckboxState_9_#4=~
     )) -> (ReturnValue:Text)
    self:Rich Text Block Object Reference=Get PropertyName() -> (PropertyName:Rich Text Block Object Reference)
   )
   n15 = Get SaveCheckbox()
   SetCheckedState(execute=@n13.else, self=@n15, InCheckedState=Unchecked)
   > n16 = SetVisibility(self=@n15, InVisibility=Hidden)
  ]}
 OnCheckboxChanged {nodes [OnCheckboxChanged!() -> (then, EntryData:Cosmetic Property Data Entry Structure) {is=title K2Node_FunctionEntry}]}
]
variables [
 EntryData {type=CosmeticPropertyDataEntry, default {allowSaving=true, lastCheckboxState=true, propertyName=None, useSingleProperty=true}}
 ItemDefsToShow {type="array<FortItemDefinition>", default []}
 ItemDefToCopy {type=FortItemDefinition, default=null}
 OnCheckboxChanged {default=()}
]
widgets {
 info {parentClass=/Script/FortniteEditor.FortCosmeticsInstrumentsEntry, rootWidgetClass=&HorizontalBox, widgetCount=3}
 widgets [
  HorizontalBox {widgetClassPath=&HorizontalBox, bIsVariable=true, widget=HorizontalBox}
  PropertyName {
   widgetClassPath=/Script/UMG.RichTextBlock
   bIsVariable=true
   parent=HorizontalBox
   widget=PropertyName
   properties {text="[PropertyName]", textStyleSet=/Game/UI/ItemReceived/AthenaFrontendStyles}
   slot {class=HorizontalBoxSlot, properties {size {sizeRule=Fill, value=0.5}}}
  }
  SaveCheckbox {
   widgetClassPath=&EditorUtilityCheckBox
   bIsVariable=true
   parent=HorizontalBox
   widget=SaveCheckbox
   properties {checkedState=Checked}
   slot {class=HorizontalBoxSlot, properties {size {sizeRule=Fill, value=0.1}}}
  }
 ]
}
