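# Editor utility widget for bulk-editing cosmetic instrument data by copying selected properties from a source Fort item definition onto target definitions.
# A self-bound details view shows the item definitions, and a button fills the targets from Content Browser selection.
# Refresh rebuilds per-property entry widgets from Properties, which checkbox changes update.
# Execute calls parent-class CopyItemDefToApplicants, while GetAllVariablesToCopy gathers checked property names.
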EditorUtilityWidgetBlueprint /Game/ArtTools/CosmeticTech/EUW_CosmeticsInstrumentsDataTool
graphs [
 AddListEntry {nodes [
   n1 = AddListEntry!() -> (then, DataEntry:Cosmetic Property Data Entry Structure) {is=title K2Node_FunctionEntry}
   > n2 = Construct Object from Class!(
    execute
    Class=&/EUW_CosmeticInstrumentsEntry.EUW_CosmeticInstrumentsEntry_C
    self=Get EntriesVerticalBox() -> (EntriesVerticalBox:Vertical Box Object Reference)
    ItemDefsToShow:Array of Fort Item Definition Object References=Get ItemDefinitionsToApplyOverrides() -> (ItemDefinitionsToApplyOverrides:Array of Fort Item Definition Object References)
    EntryData:Cosmetic Property Data Entry Structure=@n1
    ItemDefToCopy:Fort Item Definition Object Reference=Get ItemDefToCopy() -> (ItemDefToCopy:Fort Item Definition Object Reference)
   ) -> (then, ReturnValue:EUW Cosmetic Instruments Entry Object Reference)
   > Set EntryData!(
    execute
    EntryData:Cosmetic Property Data Entry Structure=Get DataEntry() -> (DataEntry:Cosmetic Property Data Entry Structure)
    self:EUW Cosmetic Instruments Entry Object Reference=@n2
   ) -> (then, Output_Get:Cosmetic Property Data Entry Structure)
   > AddChild(self=Get EntriesVerticalBox(), Content=@n2)
   > SetPadding!(
    execute
    self:User Widget Object Reference=@n2
    InPadding_Left:float=0
    InPadding_Top:float=0
    InPadding_Right:float=0
    InPadding_Bottom:float=15
   ) -> (then)
   > Bind Event to On Checkbox Changed(self:EUW Cosmetic Instruments Entry Object Reference=@n2, Delegate=Create Event()) {is=title K2Node_AddDelegate}
  ]}
 EventGraph {nodes [
   "On Clicked (Refresh)"() {is=title K2Node_ComponentBoundEvent}
   > RefreshItemDefinitionsSelected!(execute, self, PropertyChanged:Name=None) -> (then)
   "On Clicked (Execute)"() {is=title K2Node_ComponentBoundEvent}
   > CopyItemDefToApplicants(self:Fort Cosmetics Instruments Data Tool Object Reference)
   > RefreshItemDefinitionsSelected(PropertyChanged=None)
   "On Clicked (GetAssets)"() {is=title K2Node_ComponentBoundEvent}
   > Clear(TargetArray:Array of Fort Item Definition Object References=Get ItemDefinitionsToApplyOverrides())
   > n1 = GetSelectedAssets()
   > n2 = For Each Loop(Array:Array of Object References=@n1) -> (Array Element:Object Reference)
   RefreshItemDefinitionsSelected(execute=@n2.Completed, PropertyChanged=None)
   n3 = Cast To FortItemDefinition!(execute=@n2.LoopBody, Object=@n2) -> (then, CastFailed, AsFort Item Definition:Fort Item Definition Object Reference) {is=title K2Node_DynamicCast}
   > Add(
    TargetArray:Array of Fort Item Definition Object References=Get ItemDefinitionsToApplyOverrides()
    NewItem:"Fort Item Definition Object Reference (by ref)"=@n3
   ) {is=title K2Node_CallArrayFunction}
   n4 = Event PreConstruct()
   SetObject(execute=[@n4.then, @n5.then], self=Get ItemsToModifyView() -> (ItemsToModifyView:Details View Object Reference), NewObject=Self-Reference())
   n5 = Event Construct()
   Branch(Condition=false)
   > RefreshItemDefinitionsSelected(PropertyChanged=None)
  ]}
 GetAllVariablesToCopy {nodes [
   GetAllVariablesToCopy() {is=title K2Node_FunctionEntry}
   > n1 = For Each Loop(Array:Array of Widget Object References=GetAllChildren(self=Get EntriesVerticalBox())) -> (Array Element:Widget Object Reference)
   Return Node!(execute=@n1.Completed, ReturnValue:Array of Names=Get VariablesToSave() -> (VariablesToSave:Array of Names))
   > n2 = Cast To EUW_CosmeticInstrumentsEntry!(execute, Object=@n1) -> (then, CastFailed, AsEUW Cosmetic Instruments Entry:EUW Cosmetic Instruments Entry Object Reference) {is=title K2Node_DynamicCast}
   > Branch(Condition=IsChecked(self=Get SaveCheckbox!(self:EUW Cosmetic Instruments Entry Object Reference=@n2) -> (SaveCheckbox:Editor Utility Check Box Object Reference)))
   > Add Unique(
    TargetArray:Array of Names=Get VariablesToSave()
    NewItem:"Name (by ref)"=Get!(self:EUW Cosmetic Instruments Entry Object Reference=@n2) -> (
     EntryData_PropertyName_2_#1:Name
     EntryData_UseSingleProperty_5_#2:bool=~
     EntryData_AllowSaving_7_#3:bool=~
     EntryData_LastCheckboxState_9_#4:bool=~
    )
   )
  ]}
 OnEntryCheckBoxChanged {nodes [
   OnEntryCheckBoxChanged!() -> (then, EntryData:Cosmetic Property Data Entry Structure) {is=title K2Node_FunctionEntry}
   > n1 = For Each Loop!(
    Exec
    Array:Array of Cosmetic Property Data Entry Structures=Get Properties() -> (Properties:Array of Cosmetic Property Data Entry Structures)
   ) -> (
    LoopBody
    Array Element_PropertyName_2_#1:Name
    Array Element_UseSingleProperty_5_#2:bool=~
    Array Element_AllowSaving_7_#3:bool=~
    Array Element_LastCheckboxState_9_#4:bool=~
    Array Index
    Completed
   )
   > Branch(Condition="Equal (Name)"(
     A=@n1
     B=Get!() -> (
      EntryData_PropertyName_2_#1
      EntryData_UseSingleProperty_5_#2=~
      EntryData_AllowSaving_7_#3=~
      EntryData_LastCheckboxState_9_#4=~
     )
    ))
   > n2 = Set members in Cosmetic Property Data Entry!(
    execute:Exec
    StructRef:"Cosmetic Property Data Entry Structure (by ref)"="Get (a ref)"(Array:Array of Cosmetic Property Data Entry Structures=Get Properties(), Dimension 1=@n1.Array Index) -> (Output:"Cosmetic Property Data Entry Structure (by ref)")
   ) -> (then:Exec, StructOut:"Cosmetic Property Data Entry Structure (by ref)") {is=title K2Node_SetFieldsInStruct}
   > "Set Cosmetic Property Data Entry Structure (by ref)"!(
    execute:Exec
    Target:"Cosmetic Property Data Entry Structure (by ref)"=@n2
    Value:Cosmetic Property Data Entry Structure=Get EntryData() -> (EntryData:Cosmetic Property Data Entry Structure)
   ) -> (then:Exec) {is=title K2Node_VariableSetRef}
  ]}
 RefreshEntryList {nodes [
   RefreshEntryList() {is=title K2Node_FunctionEntry}
   > ClearChildren(self=Get EntriesVerticalBox())
   > n1 = For Each Loop!(Exec, Array=Get Properties()) -> (LoopBody, Array Element:Cosmetic Property Data Entry Structure, Array Index, Completed)
   > AddListEntry!(execute, self, DataEntry:Cosmetic Property Data Entry Structure=@n1) -> (then) {is=title K2Node_CallFunction}
  ]}
 RefreshItemDefinitionsSelected {nodes [
   RefreshItemDefinitionsSelected!() -> (then, PropertyChanged:Name) {is=title K2Node_FunctionEntry}
   > n1 = Sequence()
   RefreshEntryList(execute=@n1.then_1) {is=title K2Node_CallFunction}
   n2 = Get ItemDefinitionsToApplyOverrides()
   RefreshStaticWidgetsColumnSpan!(
    execute=@n1.then_0
    self
    MaxColumnSize:int=int + int(A=Length(TargetArray:Array of Fort Item Definition Object References=@n2), B=Get InitialColumnCount() -> (InitialColumnCount:int=0))
   ) -> (then)
  ]}
 RefreshStaticWidgetsColumnSpan {nodes [
   RefreshStaticWidgetsColumnSpan!() -> (then, MaxColumnSize:int) {is=title K2Node_FunctionEntry}
   > n1 = For Each Loop(Array:Array of Widget Object References=Make Array!(
     "[0]":Widget Object Reference=Get ItemsToModifyView()
     "[1]":Widget Object Reference=Get Refresh() -> (Refresh:Editor Utility Button Object Reference)
     "[2]":Widget Object Reference=Get Execute() -> (Execute:Editor Utility Button Object Reference)
     "[3]":Widget Object Reference=Get EntriesVerticalBox()
    ) -> (Array:Array of Widget Object References)) -> (Array Element:Widget Object Reference)
   > Is Valid(InputObject=@n1)
   > SetGridSlotColumnAndRow!(
    execute
    Widget:Widget Object Reference=@n1
    Row:int=Get Row!(self:Grid Slot Object Reference=SlotAsGridSlot(Widget=@n1)) -> (Row:int=0)
    Row Span:int=0
    Column:int=0
    Column Span:int=Get MaxColumnSize() -> (MaxColumnSize:int=0)
    __WorldContext:Object Reference
   ) -> (then)
  ]}
]
variables [
 InitialColumnCount {type=integer, default=2}
 Properties {
  type="array<CosmeticPropertyDataEntry>"
  default [
   {allowSaving=false, lastCheckboxState=true, propertyName=EditorTemplateId, useSingleProperty=true}
   {allowSaving=true, lastCheckboxState=true, propertyName=ItemName, useSingleProperty=true}
   {allowSaving=true, lastCheckboxState=true, propertyName=ItemDescription, useSingleProperty=true}
   {allowSaving=true, lastCheckboxState=true, propertyName=ItemShortDescription, useSingleProperty=true}
   {allowSaving=true, lastCheckboxState=true, propertyName=Icon, useSingleProperty=false}
   {allowSaving=true, lastCheckboxState=true, propertyName=LargeIcon, useSingleProperty=false}
   {allowSaving=true, lastCheckboxState=true, propertyName=Tags, useSingleProperty=false}
   {allowSaving=true, lastCheckboxState=true, propertyName=ItemVariants, useSingleProperty=false}
   {allowSaving=true, lastCheckboxState=true, propertyName=FortReleaseVersion, useSingleProperty=true}
   {allowSaving=true, lastCheckboxState=true, propertyName=Series, useSingleProperty=false}
  ]
 }
]
widgets {
 info {
  inheritedWidgetCount=1
  parentClass=/Script/FortniteEditor.FortCosmeticsInstrumentsDataTool
  rootWidgetClass=&CanvasPanel
  widgetCount=20
 }
 widgets [
  CanvasPanel_22 {widgetClassPath=&CanvasPanel, widget=CanvasPanel_22}
  EditorUtilityScrollBox_0 {
   widgetClassPath=&EditorUtilityScrollBox
   parent=CanvasPanel_22
   widget=EditorUtilityScrollBox_0
   properties {
    bAnimateWheelScrolling=true
    bOverride_Cursor=true
    widgetBarStyle {
     thickness=8
     draggedThumbImage {
      bIsDynamicallyLoaded=false
      drawAs=RoundedBox
      imageType=NoImage
      mirroring=NoMirror
      resourceName=None
      resourceObject=null
      tiling=NoTile
      imageSize=<0 0>
      margin=<ltrb 0 0 0 0>
      outlineSettings {
       bUseBrushTransparency=false
       roundingType=FixedRadius
       width=0
       color {colorUseRule=UseColor_Specified, specifiedColor=<rgba 0 0 0 0>}
       cornerRadii=<4 4 4 4>
      }
      tintColor {colorUseRule=UseColor_Specified, specifiedColor=<rgba 0.2159 0.2159 0.2159 1>}
      uVRegion {bIsValid=false, max=<0 0>, min=<0 0>}
     }
     horizontalBackgroundImage {
      bIsDynamicallyLoaded=false
      drawAs=NoDrawType
      imageType=NoImage
      mirroring=NoMirror
      resourceName=None
      resourceObject=null
      tiling=NoTile
      imageSize=<0 0>
      margin=<ltrb 0 0 0 0>
      outlineSettings {
       bUseBrushTransparency=false
       roundingType=HalfHeightRadius
       width=0
       color {colorUseRule=UseColor_Specified, specifiedColor=<rgba 0 0 0 0>}
       cornerRadii=<0 0 0 0>
      }
      tintColor {colorUseRule=UseColor_Specified, specifiedColor=<rgba 1 1 1 1>}
      uVRegion {bIsValid=false, max=<0 0>, min=<0 0>}
     }
     horizontalBottomSlotImage {
      bIsDynamicallyLoaded=false
      drawAs=NoDrawType
      imageType=NoImage
      mirroring=NoMirror
      resourceName=None
      resourceObject=null
      tiling=NoTile
      imageSize=<0 0>
      margin=<ltrb 0 0 0 0>
      outlineSettings {
       bUseBrushTransparency=false
       roundingType=HalfHeightRadius
       width=0
       color {colorUseRule=UseColor_Specified, specifiedColor=<rgba 0 0 0 0>}
       cornerRadii=<0 0 0 0>
      }
      tintColor {colorUseRule=UseColor_Specified, specifiedColor=<rgba 1 1 1 1>}
      uVRegion {bIsValid=false, max=<0 0>, min=<0 0>}
     }
     horizontalTopSlotImage {
      bIsDynamicallyLoaded=false
      drawAs=NoDrawType
      imageType=NoImage
      mirroring=NoMirror
      resourceName=None
      resourceObject=null
      tiling=NoTile
      imageSize=<0 0>
      margin=<ltrb 0 0 0 0>
      outlineSettings {
       bUseBrushTransparency=false
       roundingType=HalfHeightRadius
       width=0
       color {colorUseRule=UseColor_Specified, specifiedColor=<rgba 0 0 0 0>}
       cornerRadii=<0 0 0 0>
      }
      tintColor {colorUseRule=UseColor_Specified, specifiedColor=<rgba 1 1 1 1>}
      uVRegion {bIsValid=false, max=<0 0>, min=<0 0>}
     }
     hoveredThumbImage {
      bIsDynamicallyLoaded=false
      drawAs=RoundedBox
      imageType=NoImage
      mirroring=NoMirror
      resourceName=None
      resourceObject=null
      tiling=NoTile
      imageSize=<0 0>
      margin=<ltrb 0 0 0 0>
      outlineSettings {
       bUseBrushTransparency=false
       roundingType=FixedRadius
       width=0
       color {colorUseRule=UseColor_Specified, specifiedColor=<rgba 0 0 0 0>}
       cornerRadii=<4 4 4 4>
      }
      tintColor {colorUseRule=UseColor_Specified, specifiedColor=<rgba 0.2159 0.2159 0.2159 1>}
      uVRegion {bIsValid=false, max=<0 0>, min=<0 0>}
     }
     normalThumbImage {
      bIsDynamicallyLoaded=false
      drawAs=RoundedBox
      imageType=NoImage
      mirroring=NoMirror
      resourceName=None
      resourceObject=null
      tiling=NoTile
      imageSize=<0 0>
      margin=<ltrb 0 0 0 0>
      outlineSettings {
       bUseBrushTransparency=false
       roundingType=FixedRadius
       width=0
       color {colorUseRule=UseColor_Specified, specifiedColor=<rgba 0 0 0 0>}
       cornerRadii=<4 4 4 4>
      }
      tintColor {colorUseRule=UseColor_Specified, specifiedColor=<rgba 0.0953 0.0953 0.0953 1>}
      uVRegion {bIsValid=false, max=<0 0>, min=<0 0>}
     }
     verticalBackgroundImage {
      bIsDynamicallyLoaded=false
      drawAs=NoDrawType
      imageType=NoImage
      mirroring=NoMirror
      resourceName=None
      resourceObject=null
      tiling=NoTile
      imageSize=<0 0>
      margin=<ltrb 0 0 0 0>
      outlineSettings {
       bUseBrushTransparency=false
       roundingType=HalfHeightRadius
       width=0
       color {colorUseRule=UseColor_Specified, specifiedColor=<rgba 0 0 0 0>}
       cornerRadii=<0 0 0 0>
      }
      tintColor {colorUseRule=UseColor_Specified, specifiedColor=<rgba 1 1 1 1>}
      uVRegion {bIsValid=false, max=<0 0>, min=<0 0>}
     }
     verticalBottomSlotImage {
      bIsDynamicallyLoaded=false
      drawAs=NoDrawType
      imageType=NoImage
      mirroring=NoMirror
      resourceName=None
      resourceObject=null
      tiling=NoTile
      imageSize=<0 0>
      margin=<ltrb 0 0 0 0>
      outlineSettings {
       bUseBrushTransparency=false
       roundingType=HalfHeightRadius
       width=0
       color {colorUseRule=UseColor_Specified, specifiedColor=<rgba 0 0 0 0>}
       cornerRadii=<0 0 0 0>
      }
      tintColor {colorUseRule=UseColor_Specified, specifiedColor=<rgba 1 1 1 1>}
      uVRegion {bIsValid=false, max=<0 0>, min=<0 0>}
     }
     verticalTopSlotImage {
      bIsDynamicallyLoaded=false
      drawAs=NoDrawType
      imageType=NoImage
      mirroring=NoMirror
      resourceName=None
      resourceObject=null
      tiling=NoTile
      imageSize=<0 0>
      margin=<ltrb 0 0 0 0>
      outlineSettings {
       bUseBrushTransparency=false
       roundingType=HalfHeightRadius
       width=0
       color {colorUseRule=UseColor_Specified, specifiedColor=<rgba 0 0 0 0>}
       cornerRadii=<0 0 0 0>
      }
      tintColor {colorUseRule=UseColor_Specified, specifiedColor=<rgba 1 1 1 1>}
      uVRegion {bIsValid=false, max=<0 0>, min=<0 0>}
     }
    }
   }
   slot {class=CanvasPanelSlot, properties {layoutData {alignment=<0 0>, anchors {maximum=<1 1>, minimum=<0 0>}, offsets=<ltrb 0 0 0 0>}}}
  }
  VerticalBox_81 {widgetClassPath=&VerticalBox, parent=EditorUtilityScrollBox_0, widget=VerticalBox_81, slot {class=ScrollBoxSlot}}
  HorizontalBox_0 {
   widgetClassPath=&HorizontalBox
   parent=VerticalBox_81
   widget=HorizontalBox_0
   slot {class=VerticalBoxSlot, properties {padding=<ltrb 0 0 0 15>}}
  }
  AutoOpenEditorAssetsOnceModified {
   widgetClassPath=&EditorUtilityCheckBox
   bInherited=true
   bIsVariable=true
   parent=HorizontalBox_0
   widget=AutoOpenEditorAssetsOnceModified
   properties {checkedState=Checked}
   slot {class=HorizontalBoxSlot}
  }
  ScaleBox_0 {
   widgetClassPath=&ScaleBox
   parent=HorizontalBox_0
   widget=ScaleBox_0
   properties {stretch=UserSpecified, userSpecifiedScale=0.75}
   slot {class=HorizontalBoxSlot, properties {padding=<ltrb 15 0 0 0>}}
  }
  AutoOpenText {
   widgetClassPath=/Script/UMG.RichTextBlock
   parent=ScaleBox_0
   widget=AutoOpenText
   properties {text=Auto Open Assets once Modified, textStyleSet=/Game/UI/ItemReceived/AthenaFrontendStyles}
   slot {class=ScaleBoxSlot}
  }
  GridPanel {
   widgetClassPath=&GridPanel
   bIsVariable=true
   parent=VerticalBox_81
   widget=GridPanel
   properties {columnFill [1,1,1,1,1]}
   slot {class=VerticalBoxSlot, properties {size {sizeRule=Fill, value=1}}}
  }
  ItemsToModifyView {
   widgetClassPath=&DetailsView
   parent=GridPanel
   widget=ItemsToModifyView
   properties {categoriesToShow [Item Definitions]}
   slot {class=GridSlot, properties {columnSpan=5, padding=<ltrb 0 0 0 15>}}
  }
  HorizontalBox_1 {
   widgetClassPath=&HorizontalBox
   parent=GridPanel
   widget=HorizontalBox_1
   slot {class=GridSlot, properties {columnSpan=5, row=1, padding=<ltrb 0 0 0 15>}}
  }
  Refresh {
   widgetClassPath=&EditorUtilityButton
   parent=HorizontalBox_1
   widget=Refresh
   slot {class=HorizontalBoxSlot, properties {size {sizeRule=Fill, value=1}}}
  }
  RefreshText {
   widgetClassPath=&RichTextBlock
   parent=Refresh
   widget=RefreshText
   properties {text=Refresh, textStyleSet=&AthenaFrontendStyles}
   slot {class=ButtonSlot}
  }
  GetAssets {
   widgetClassPath=&EditorUtilityButton
   parent=HorizontalBox_1
   widget=GetAssets
   slot {class=HorizontalBoxSlot, properties {size {sizeRule=Fill, value=1}}}
  }
  RefreshText_1 {
   widgetClassPath=&RichTextBlock
   parent=GetAssets
   widget=RefreshText_1
   properties {text=Get IDs From Browser, textStyleSet=&AthenaFrontendStyles}
   slot {class=ButtonSlot}
  }
  EditorUtilityScrollBox_1 {
   widgetClassPath=&EditorUtilityScrollBox
   parent=GridPanel
   widget=EditorUtilityScrollBox_1
   properties {orientation=Orient_Horizontal}
   slot {class=GridSlot, properties {columnSpan=5, row=3, rowSpan=5}}
  }
  EntriesVerticalBox {
   widgetClassPath=&VerticalBox
   bIsVariable=true
   parent=EditorUtilityScrollBox_1
   widget=EntriesVerticalBox
   slot {class=ScrollBoxSlot, properties {size {sizeRule=Fill, value=1}}}
  }
  SingleViewHolder {widgetClassPath=&VerticalBox, parent=GridPanel, widget=SingleViewHolder, slot {class=GridSlot}}
  CustomPropertyView_191 {
   widgetClassPath=/Script/FortniteEditor.CustomPropertyView
   bIsVariable=true
   parent=GridPanel
   widget=CustomPropertyView_191
   slot {class=GridSlot}
  }
  Execute {
   widgetClassPath=&EditorUtilityButton
   parent=VerticalBox_81
   widget=Execute
   slot {class=VerticalBoxSlot, properties {verticalAlignment=VAlign_Top, padding=<ltrb 0 25 0 0>}}
  }
  Text {
   widgetClassPath=&RichTextBlock
   parent=Execute
   widget=Text
   properties {justification=Center, text=Override Selected Values, textStyleSet=&AthenaFrontendStyles}
   slot {class=ButtonSlot}
  }
 ]
}
